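# Sound cue for the medkit drop sound in Athena, routed to the loot collect sound class with weapon drop attenuation.
# A random node picks one of four equally weighted drop waves and is excluded from branch culling.
# It feeds a mixer whose other input is empty, then a root modulator that slightly varies pitch on each play.
# The cue plays at reduced volume.
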SoundCue /Game/Athena/Sounds/Items/Medkit/Medkit_Drop_Cue
root=Modulator_1
nodes [
 {id=Mixer_0, children [null,Random_0], properties {inputVolume [1,1]}}
 {id=Modulator_1, children [Mixer_0], properties {pitchMax=1.1, pitchMin=0.9}}
 {
  id=Random_0
  children [WavePlayer_1,WavePlayer_2,WavePlayer_3,WavePlayer_4]
  properties {bShouldExcludeFromBranchCulling=true, weights [1,1,1,1]}
 }
 {id=WavePlayer_1, properties {soundWaveAssetPtr=&/Medkit_Drop_01}}
 {id=WavePlayer_2, properties {soundWaveAssetPtr=&/Medkit_Drop_02}}
 {id=WavePlayer_3, properties {soundWaveAssetPtr=&/Medkit_Drop_03}}
 {id=WavePlayer_4, properties {soundWaveAssetPtr=&/Medkit_Drop_04}}
]
properties {
 attenuationSettings=&Fort_Weapon_Drop_Att
 maxDistance=3000
 soundClassObject=&Fort_Loot_Collect
 volumeMultiplier=0.4
}
