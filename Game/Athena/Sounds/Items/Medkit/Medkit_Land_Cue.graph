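# Sound cue for a medkit landing, using weapon-drop attenuation and concurrency settings and the loot-collect sound class.
# A mixer layers one of four randomly chosen landing variations with a single thunk wave at a lower input volume.
# Each layer passes through its own modulator for pitch variation, and the thunk layer also varies in volume.
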SoundCue /Game/Athena/Sounds/Items/Medkit/Medkit_Land_Cue
root=Mixer_0
nodes [
 {id=Mixer_0, children [Modulator_1,Modulator_0], properties {inputVolume [1,0.7]}}
 {id=Modulator_0, children [WavePlayer_8], properties {pitchMax=1.2, pitchMin=0.9, volumeMax=1, volumeMin=0.8}}
 {id=Modulator_1, children [Random_1], properties {pitchMax=1.1, pitchMin=0.9}}
 {
  id=Random_1
  children [WavePlayer_4,WavePlayer_5,WavePlayer_6,WavePlayer_7]
  properties {bShouldExcludeFromBranchCulling=true, weights [1,1,1,1]}
 }
 {id=WavePlayer_4, properties {soundWaveAssetPtr=&/Medkit_Land_01}}
 {id=WavePlayer_5, properties {soundWaveAssetPtr=&/Medkit_Land_02}}
 {id=WavePlayer_6, properties {soundWaveAssetPtr=&/Medkit_Land_03}}
 {id=WavePlayer_7, properties {soundWaveAssetPtr=&/Medkit_Land_04}}
 {id=WavePlayer_8, properties {soundWaveAssetPtr=&/Medkit_Land_Thunk_01}}
]
properties {
 attenuationSettings=&Fort_Weapon_Drop_Att
 maxDistance=3000
 soundClassObject=&Fort_Loot_Collect
 volumeMultiplier=0.7
 concurrencyOverrides {resolutionRule=PreventNew}
 concurrencySet [&Weapon_Drop_StopOldest]
}
