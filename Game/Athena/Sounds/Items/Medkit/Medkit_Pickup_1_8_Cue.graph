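# Sound cue for a Medkit pickup in the Athena mode, built from reused front-end UI sounds.
# A mixer layers a big hit and a menu-appear wave, plus a dress-up wave that a short fixed delay offsets.
# It uses the weapon pickup attenuation and loot-collect sound class at reduced volume.
# A concurrency override prevents new instances from starting once the limit is reached.
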SoundCue /Game/Athena/Sounds/Items/Medkit/Medkit_Pickup_1_8_Cue
root=Mixer_0
nodes [
 {id=Delay_0, children [WavePlayer_3], properties {delayMax=0.1, delayMin=0.1}}
 {id=Mixer_0, children [WavePlayer_2,WavePlayer_1,Delay_0], properties {inputVolume [1,1,1]}}
 {id=WavePlayer_1, properties {soundWaveAssetPtr=/Game/Sounds/Fort_UI_Sounds/FrontEnd/Fort_Menu_Appear_02}}
 {id=WavePlayer_2, properties {soundWaveAssetPtr=/Game/Sounds/Fort_UI_Sounds/FrontEnd/fort_ui_big_hit_01}}
 {id=WavePlayer_3, properties {soundWaveAssetPtr=/Game/Sounds/Fort_UI_Sounds/FrontEnd/fort_ui_dressup_01}}
]
properties {
 attenuationSettings=&Fort_Weapon_Pickup_Att
 bOverrideConcurrency=true
 maxDistance=3500
 soundClassObject=&Fort_Loot_Collect
 volumeMultiplier=0.5
 concurrencyOverrides {resolutionRule=PreventNew}
}
