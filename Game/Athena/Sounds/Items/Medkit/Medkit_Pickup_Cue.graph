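# Sound cue that plays when a medkit item is picked up.
# A single pickup wave player feeds a Modulator root, which randomizes pitch for variation between plays.
# It uses the loot-collect sound class and shared weapon-pickup attenuation.
# A concurrency override blocks new instances once the limit is reached instead of stopping playing ones.
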SoundCue /Game/Athena/Sounds/Items/Medkit/Medkit_Pickup_Cue
root=Modulator_0
nodes [
 {id=Modulator_0, children [WavePlayer_0], properties {pitchMax=1.2}}
 {id=WavePlayer_0, properties {soundWaveAssetPtr=&/Medkit_Pickup_01}}
]
properties {
 attenuationSettings=&Fort_Weapon_Pickup_Att
 bOverrideConcurrency=true
 maxDistance=3500
 soundClassObject=&Fort_Loot_Collect
 volumeMultiplier=0.8
 concurrencyOverrides {resolutionRule=PreventNew}
}
